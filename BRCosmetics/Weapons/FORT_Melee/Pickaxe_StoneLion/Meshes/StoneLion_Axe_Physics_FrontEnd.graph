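# Physics asset for the StoneLion two-handed pickaxe, a front-end variant by its name, that sets up a dangling chain on the axe.
# Five capsules form the collision shape of the handle, and smaller capsules cover a master chain bone and three dynamic segments that grow thicker toward the tip.
# Constraints link the chain in series from the handle, with free rotation at the root and angle limits between segments, looser toward the tip.
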PhysicsAsset /BRCosmetics/Weapons/FORT_Melee/Pickaxe_StoneLion/Meshes/StoneLion_Axe_Physics_FrontEnd
bodies {
 MSTR_chain [{shapeType=Capsule, length=0.505, radius=0.505, shapeName=MSTR_chain_capsule, rotation=<rot 0 -90 0>}]
 Melee_TwoHanded_Handle [
  {shapeType=Capsule, length=10, radius=2, center=<0 -3.6828 -2.8845>, rotation=<rot -0.0002 0 -10>}
  {
   shapeType=Capsule
   length=10.5
   radius=1
   shapeName=Melee_TwoHanded_Handle_capsule
   center=<-1.0589 3.9849 -10.3081>
   rotation=<rot -0.0001 0 -90>
  }
  {
   shapeType=Capsule
   length=10.5
   radius=1
   shapeName=Melee_TwoHanded_Handle_capsule
   center=<-1.0589 2.5917 -8.3524>
   rotation=<rot -0.0001 0 -90>
  }
  {
   shapeType=Capsule
   length=10.5
   radius=1
   shapeName=Melee_TwoHanded_Handle_capsule
   center=<-1.0589 5.313 -12.3524>
   rotation=<rot -0.0001 0 -90>
  }
  {shapeType=Capsule, length=10, radius=2}
 ]
 dyn_chain_01 [{shapeType=Capsule, length=3, radius=0.7, shapeName=dyn_chain_01_capsule, center=<4.5965 0.6984 0>, rotation=<rot -90 0 0>}]
 dyn_chain_02 [{shapeType=Capsule, length=3, radius=1.5, shapeName=dyn_chain_02_capsule, center=<3 0 0>, rotation=<rot -90 0 0>}]
 dyn_chain_03 [{shapeType=Capsule, length=3, radius=2, shapeName=dyn_chain_03_capsule, center=<3 0 0>, rotation=<rot -90 0 0>}]
}
constraints [
 {
  bone1Name=MSTR_chain
  bone2Name=Melee_TwoHanded_Handle
  swing1LimitDegrees=45
  swing1Motion=Free
  swing2LimitDegrees=45
  swing2Motion=Free
  twistLimitDegrees=45
  twistMotion=Free
 }
 {bone1Name=dyn_chain_01, bone2Name=MSTR_chain, swing1LimitDegrees=45, swing2LimitDegrees=45, twistLimitDegrees=60}
 {bone1Name=dyn_chain_02, bone2Name=dyn_chain_01, swing1LimitDegrees=60, swing2LimitDegrees=60, twistLimitDegrees=60}
 {bone1Name=dyn_chain_03, bone2Name=dyn_chain_02, swing1LimitDegrees=60, swing2LimitDegrees=60, twistLimitDegrees=60}
]
